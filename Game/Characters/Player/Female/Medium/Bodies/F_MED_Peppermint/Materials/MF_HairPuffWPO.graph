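# Material function for the Peppermint character body that puffs hair outward through World Position Offset.
# It adds the world vertex normal, scaled by HairOffsetMultiplier and the mask texture's red channel, to the incoming attributes' existing offset.
# The Use HairPuffWPO static switch selects this result or passes the input through unchanged.
# A camera-relative directional term with a HairOffsetDirection vector is computed but left unconnected to the output.
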MaterialFunction /Game/Characters/Player/Female/Medium/Bodies/F_MED_Peppermint/Materials/MF_HairPuffWPO
graph {nodes [
  n1 = FunctionInput() {inputType=FunctionInput_MaterialAttributes}
  n2 = GetMaterialAttributes(None=@n1) {attributeGetTypes [#1]}
  n3 = ScalarParameter() {defaultValue=1, group=HairPuffWPO, parameterName=HairOffsetMultiplier}
  n4 = VertexNormalWS()
  n5 = TextureSampleParameter2D() {
   group=HairPuffWPO
   parameterName=Mask Texture
   samplerType=SAMPLERTYPE_Masks
   sortPriority=1
   texture=&T_BlackMask
  }
  n6 = Add(A=@n2.WorldPositionOffset, B=Multiply(A=Multiply(A=@n3, B=@n4), B=@n5.R))
  FunctionOutput(None=StaticSwitchParameter(
    True=SetMaterialAttributes!(MaterialAttributes=@n2.MaterialAttributes, WorldPositionOffset=@n6) {attributeSetTypes [#1]}
    False=@n1
   ) {group=Skin FX, parameterName=Use HairPuffWPO})
  n7 = WorldPosition()
  n8 = ComponentMask(None=Subtract(A=@n7.XYZ, B=ActorPositionWS())) {g=true, r=true}
  n9 = CameraPositionWS()
  n10 = ComponentMask(None=Normalize(VectorInput=Subtract(A=@n7.XYZ, B=@n9))) {g=true, r=true}
  n11 = ConstantBiasScale(None=DotProduct(A=Normalize(VectorInput=@n8), B=Normalize(VectorInput=@n10)))
  Multiply(A=@n4, B=@n11)
  Multiply(A=@n11, B=VectorParameter() {group=HairPuffWPO, parameterName=HairOffsetDirection, defaultValue=<rgba 5 5 2 1>}.RGB)
 ]}
